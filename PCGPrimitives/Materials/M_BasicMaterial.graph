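# M_BasicMaterial is a simple Default Lit surface material in PCGPrimitives, flagged for use with Nanite, instanced static meshes and spline meshes.
# A static switch optionally blends the base color toward a dirt color using the blue channel of a virtual-texture packed mask.
# A scaled per-instance random value is added to the color to vary brightness between instances.
# Roughness and specular are exposed as scalar parameters.
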Material /PCGPrimitives/Materials/M_BasicMaterial
graph {
 properties {
  bUsedWithInstancedStaticMeshes=true
  bUsedWithNanite=true
  bUsedWithSplineMeshes=true
  pixelDepthOffsetMode=PDOM_Legacy
  usedShadingModels=Default Lit
 }
 nodes [
  n1 = Multiply(A=PerInstanceRandom(), B=ScalarParameter() {defaultValue=0.01, parameterName=per random add color}) {constA=0.1, constB=0.01}
  Add(A=@n1) {constB=-0.05}
  n2 = VectorParameter() {parameterName=Base Color, defaultValue=<rgba 0.0431 0.0431 0.0431 0>}
  n3 = VectorParameter() {parameterName=dirt_color, defaultValue=<rgba 0.02 0.02 0.02 0>}
  n4 = TextureSampleParameter2D(UVs=TextureCoordinate()) {
   parameterName=Base_Pattern
   samplerType=SAMPLERTYPE_VirtualMasks
   texture=/PCGPrimitives/Materials/Textures/freeway_gradientLines_packed_v03a_v03_circleAlpha
  }
  Add(
   A=StaticSwitchParameter(True=LinearInterpolate(A=@n2.RGB, B=@n3.RGB, Alpha=@n4.B), False=@n2.RGB) {defaultValue=true, parameterName=Use Dirt_Texture}
   B=@n1
  ) {constB=0}
  ScalarParameter() {defaultValue=0.2, parameterName=Roughness, sliderMax=1}
  ScalarParameter() {defaultValue=0.3, parameterName=Specular, sliderMax=1}
 ]
}
